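# Niagara effect for the StoneFly emote: a glowing blue screen with floating shapes, sparkles and a flickering light.
# Shapes rise as orange and cyan SubUV sprites in three depth layers; rising and static sparkles twinkle around them.
# Screen combines meshes and glow sprites that narrow by camera angle, while a light flickers cyan to orange.
# Time-of-day helpers tune brightness; rising sparkles have a front-end spawn rate; several emitters are empty.
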NiagaraSystem /BRCosmetics/Athena/Items/Cosmetics/Dances/BacklogEmotes/BR/StoneFly/FX/NS_Emotes_StoneFly_Shapes
dependencies {
 usedDataInterfaces [
  &NiagaraDataInterfaceCurve
  &NiagaraDataInterfaceColorCurve
  &NiagaraDataInterfaceArrayColor
  &NiagaraDataInterfaceVector2DCurve
  &NiagaraDataInterfaceSpriteRendererInfo
  &NiagaraDataInterfaceArrayFloat3
  &NiagaraDataInterfaceCamera
 ]
 usedDynamicInputs [
  &MakeCustomVectorFromBool
  &RandomBool
  &FN_FrontEndVector3Modifier
  &MakeCustomLinearColorFromBool
 ]
 usedModules [
  &SystemState
  &self:SystemSpawnScript.SystemScriptSource.SystemScriptGraph.NiagaraNodeAssignment_0.SetVariables_#1
 ]
}
emitters {
 Fake_Shape_Lights {rendererClasses [], renderers []}
 Glows_Air {rendererClasses [], renderers []}
 Glows_Decal {rendererClasses [], renderers []}
 Legs_Light {rendererClasses [], renderers []}
 Rising_Sparkles {
  emitterSpawnScript [/SetVariables(Emitter.FE Particle Spawn Rate=16, Emitter.Particle Spawn Rate=13) {isSetParameters=true}]
  emitterUpdateScript [
   &EmitterState()
   &SpawnBurst_Instantaneous(Spawn Count=12, Use Spawn Probability=false, UseLoopCountLimit=false)
   &SpawnRate(
    SpawnRate=&FN_FrontEndFloatModifier(Gameplay=$Emitter.Particle Spawn Rate, Front End=$Emitter.FE Particle Spawn Rate)
    Use Spawn Probability=false
   )
  ]
  particleSpawnScript [
   /SetVariables(Particles.VisibilityTag=&RandomRangeInt(Minimum=0, Maximum=2)) {isSetParameters=true}
   &InitializeParticle(
    Color Maximum=<rgba 0.2917 0.7498 1 1>
    Color Minimum=<rgba 1 1 1 1>
    Color Mode=::NewEnumerator2
    Lifetime Max=3.15
    Lifetime Min=2.75
    Lifetime Mode=::NewEnumerator1
    Sprite Size Mode=::NewEnumerator1
    Uniform Sprite Size Max=25
    Uniform Sprite Size Min=8
    UsePositionOffset=false
    Write Material Random=true
   )
   &ShapeLocation(
    Handle Radius=30
    Large Radius=60
    MaskSpawnGroup=false
    Non Uniform Scale=<1 1 0.4>
    Offset=<0 0 16>
    Offset Mode=::NewEnumerator0
    Override Seed=false
    Rotation Mode=::NewEnumerator2
    Shape Primitive=::NewEnumerator3
    UseOwnerScale=true
    Yaw / Pitch / Roll=<1.6 0 0>
   )
   &AddVelocity(Velocity=&RandomRangeVector(Minimum=<0 0 40>, Maximum=<0 0 70>))
   &ApplyInitialForces()
  ]
  particleUpdateScript [
   &ParticleState()
   /SetVariables(
    Particles.Color=&MultiplyLinearColor(
     Linear Color=$Particles.Color
     Scale Factor=&MakeVector4_FromLinearColor(LinearColor=<rgba 8.0099 9.8007 10 1>)
    )
    Particles.Color02:LinearColor=&MultiplyLinearColor(
     Linear Color=$Particles.Color
     Scale Factor=&MakeVector4_FromLinearColor(LinearColor=<rgba 0.9441 0.8079 1 1>)
    )
    Particles.Color03:LinearColor=&MultiplyLinearColor(
     Linear Color=$Particles.Color
     Scale Factor=&MakeVector4_FromLinearColor(LinearColor=<rgba 0.15 0.15 0.15 1>)
    )
    Particles.SpriteSize02:Vector2f=&Multiply_Vector2(A=$Particles.SpriteSize, B=&Vector2DFromFloat(Value=0.3))
   ) {isSetParameters=true}
   &ScaleColor(
    Scale Alpha=&Multiply_Float(
     A=&FloatFromCurve(
      FloatCurve=DataInterface() {propertyValues {Curve {
         defaultValue=1
         keys [
          {interpMode=RCIM_Cubic, tangentMode=RCTM_User, time=0, value=0}
          {arriveTangent=1.3821, interpMode=RCIM_Cubic, leaveTangent=1.3821, time=0.0595, value=1.2282}
          {arriveTangent=1.1681, interpMode=RCIM_Cubic, leaveTangent=1.1681, time=0.1056, value=0.1459}
          {arriveTangent=1.4275, interpMode=RCIM_Cubic, leaveTangent=1.4275, time=0.2471, value=1.4474}
          {arriveTangent=3.9425, interpMode=RCIM_Cubic, leaveTangent=3.9425, time=0.2899, value=0.409}
          {arriveTangent=-6.2651, interpMode=RCIM_Cubic, leaveTangent=-6.2651, time=0.3195, value=1.7329}
          {interpMode=RCIM_Cubic, tangentMode=RCTM_User, time=0.3552, value=0}
          {arriveTangent=2.1335, interpMode=RCIM_Cubic, leaveTangent=2.1335, time=0.4117, value=1.0915}
          {arriveTangent=-3.4538, interpMode=RCIM_Cubic, leaveTangent=-3.4538, time=0.4512, value=0.2048}
          {arriveTangent=-0.8911, interpMode=RCIM_Cubic, leaveTangent=-0.8911, time=0.5083, value=0.7578}
          {arriveTangent=13.8823, interpMode=RCIM_Cubic, leaveTangent=13.8823, time=0.5622, value=0.1059}
          {arriveTangent=0.9728, interpMode=RCIM_Cubic, leaveTangent=0.9728, time=0.5794, value=1.7452}
          {arriveTangent=-9.0207, interpMode=RCIM_Cubic, leaveTangent=-9.0207, time=0.6159, value=0.1581}
          {arriveTangent=-0.8815, interpMode=RCIM_Cubic, leaveTangent=-0.8815, time=0.6638, value=0.9837}
          {arriveTangent=3.0035, interpMode=RCIM_Cubic, leaveTangent=3.0035, time=0.7869, value=0.0073}
          {arriveTangent=-0.0444, interpMode=RCIM_Cubic, leaveTangent=-0.0444, time=0.845, value=1.5277}
          {arriveTangent=8.8246, interpMode=RCIM_Cubic, leaveTangent=8.8246, tangentMode=RCTM_User, time=0.8889, value=0.0028}
          {tangentMode=RCTM_User, time=1, value=0}
         ]
        }}}
      Scale Curve=2
     )
     B=&FloatFromCurve(FloatCurve=DataInterface() {propertyValues {Curve {keys [
          <at 0.0349 2.3324 RCIM_Cubic>
          {arriveTangentWeight=0.1487, interpMode=RCIM_Cubic, leaveTangentWeight=0.3469, tangentMode=RCTM_User, time=0.233, value=1}
          {arriveTangentWeight=0.2531, interpMode=RCIM_Cubic, leaveTangentWeight=0.167, tangentMode=RCTM_User, time=0.8576, value=1}
          {arriveTangent=-21.7799, interpMode=RCIM_Cubic, leaveTangent=-21.7799, tangentMode=RCTM_User, time=1, value=0}
         ]}}})
    )
    Scale RGB=&MakeVectorFromLinearColorRGB(Color=&MultiplyLinearColor(
      Linear Color=&ColorFromCurve(ColorCurve=DataInterface() {propertyValues {
         AlphaCurve {keys []}
         BlueCurve {keys [<at 0 1 RCIM_Cubic>,<at 0.1788 1>,<at 0.207 15>,<at 0.2299 1>,<at 0.48 1>,<at 0.5029 10>,<at 0.5294 1>,<at 0.8729 1>,<at 0.9028 10>,<at 0.9275 1>,<at 0.9433 15>,<at 0.968 1>,<at 0.9874 10>,<at 1 1 RCIM_Cubic>]}
         GreenCurve {keys [<at 0 1 RCIM_Cubic>,<at 0.1788 1>,<at 0.207 15>,<at 0.2299 1>,<at 0.48 1>,<at 0.5029 10>,<at 0.5294 1>,<at 0.8729 1>,<at 0.9028 10>,<at 0.9275 1>,<at 0.9433 15>,<at 0.968 1>,<at 0.9874 10>,<at 1 1 RCIM_Cubic>]}
         RedCurve {keys [<at 0 1 RCIM_Cubic>,<at 0.1788 1>,<at 0.207 15>,<at 0.2299 1>,<at 0.48 1>,<at 0.5029 10>,<at 0.5294 1>,<at 0.8729 1>,<at 0.9028 10>,<at 0.9275 1>,<at 0.9433 15>,<at 0.968 1>,<at 0.9874 10>,<at 1 1 RCIM_Cubic>]}
        }})
      Scale Factor=&MakeVector4_FromLinearColor(LinearColor=&FNTimeOfDayColor(Night Color=<rgba 0.3 0.6267 1 10>, Morning Color=<rgba 1 1 1 1>, Day Color=<rgba 1 1 1 1>, Evening Color=<rgba 1 1 1 1>))
     ))
    ScaleA=true
    ScaleRGB=true
   )
   &CurlNoiseForce(
    Bias Noise=false
    Noise Frequency=&RandomRangeFloat(Minimum=25, Maximum=45)
    Noise Strength=&RandomRangeFloat(Minimum=30, Maximum=45)
    Pan Noise=false
   )
   &Drag(Drag=&RandomRangeFloat(Minimum=0.5, Maximum=0.75), Use Linear Drag=true, Use Rotational Drag=true)
   &SolveForcesAndVelocity(Clamp Velocity=false, Limit Acceleration=false)
   /SetVariables(
    Particles.Corona Size:Vector2f=&Multiply_Vector2D_ByFloat(Float=&RandomRangeFloat(Minimum=1.7, Maximum=2), Vector2D=$Particles.Initial.SpriteSize)
    Particles.Sharp Size:Vector2f=&Multiply_Vector2D_ByFloat(Float=&RandomRangeFloat(Minimum=1.5, Maximum=7), Vector2D=$Particles.Initial.SpriteSize)
   ) {isSetParameters=true}
  ]
  rendererClasses [&NiagaraSpriteRendererProperties]
  renderers [
   {rendererClass=&NiagaraSpriteRendererProperties}
   {rendererClass=&NiagaraSpriteRendererProperties, rendererIndex=1}
   {rendererClass=&NiagaraSpriteRendererProperties, rendererIndex=2}
  ]
 }
 Screen {
  emitterSpawnScript [/SetVariables(Emitter.U Scale Power:NiagaraFloat=&Waveform("Waveform [1]"=::NewEnumerator0, "[1] Frequency"=20, "[1] Amplitude Min/Max"=<0.5 0.6>)) {isSetParameters=true}]
  emitterUpdateScript [
   &EmitterState()
   &SpawnBurst_Instantaneous(Spawn Time=0.05, Use Spawn Probability=false, UseLoopCountLimit=false)
   &SpawnRate(SpawnRate=2.75, Use Spawn Probability=false)
   &SpawnBurst_Instantaneous(Spawn Group=1, Spawn Time=0.01, Use Spawn Probability=false, UseLoopCountLimit=false) {module=SpawnBurst_Instantaneous001}
   &SpawnBurst_Instantaneous(Spawn Group=2, Spawn Time=0.01, Use Spawn Probability=false, UseLoopCountLimit=false) {module=SpawnBurst_Instantaneous002}
  ]
  particleSpawnScript [
   /SetVariables(Particles.VisibilityTag=&SpawnGroup()) {isSetParameters=true}
   &InitializeParticle(
    Color=&MergeHelper_ColorSwitch(
     0=&RandomRangeLinearColor(Minimum=<rgba 0 0.322 2.3 0.7>, Maximum=<rgba 0 0.3201 1.8 0.63>)
     1=<rgba 0 0.2311 1 0.28>
     2=<rgba 0 0.207 1 0.045>
     3=?
     Index=&SpawnGroup()
     Num=::NewEnumerator1
    )
    Lifetime=&MergeHelper_FloatSwitch(
     0=&RandomRangeFloat(Minimum=0.7, Maximum=1)
     1=$Emitter.CurrentLoopDuration
     2=$Emitter.CurrentLoopDuration
     3=?
     Index=&SpawnGroup()
     Num=::NewEnumerator1
    )
    Mesh Scale=&RandomRangeVector(Minimum=<4 0.05 3>, Maximum=<7 0.05 5>)
    Mesh Scale Mode=::NewEnumerator4
    Position Offset=&MergeHelper_VectorSwitch(
     0=&RandomRangeVector(Minimum=<-33 -80 75>, Maximum=<33 -80 80>)
     1=<0 -80 115>
     2=<0 -80 100>
     3=?
     Index=&SpawnGroup()
     Num=::NewEnumerator1
    )
    Sprite Size=&MergeHelper_Vector2Switch(0=<0 0>, 1=<315 250>, 2=<200 160>, 3=?, 4=?, Index=&SpawnGroup(), Num=::NewEnumerator1)
    Sprite Size Mode=::NewEnumerator4
    UsePositionOffset=true
    Write Material Random=true
   )
   &SpriteFacingAndAlignment(Sprite Facing=<0 1 0>, WriteAlignment=false, WriteFacing=true)
   &ShapeLocation(
    Box Size=&MergeHelper_VectorSwitch(0=<5 15 130>, 1=<0 0 0>, 2=<0 0 0>, 3=?, Index=&SpawnGroup(), Num=::NewEnumerator1)
    MaskSpawnGroup=false
    Override Seed=false
    Shape Primitive=::NewEnumerator2
    Surface Only=false
    UseOwnerScale=true
   )
   &InitialMeshOrientation(
    Mesh Orientation Mode=::NewEnumerator2
    Orientation Vector=&SelectVectorFromArray(
     Array Sampling Mode=::NewEnumerator0
     Direct Array Index=?
     Vector Selection Array=DataInterface() {propertyValues {FloatData [<1 0 0>, <-1 0 0>]}}
     Random Seed=0
    )
    Use Rotation Vector=false
   )
  ]
  particleUpdateScript [
   &ParticleState()
   &ScaleSpriteSize(
    Scale Factor=&Multiply_Vector2D(
     A=<1 1>
     B=&MakeVector2D(
      X=&Multiply_Float(
       A=1
       B=&RemapRange(
        Input Value To Remap=&Abs_Float(Float:NiagaraFloat=&DotProduct(
          Vector A=&ReturnCameraVector(Camera Vectors=::NewEnumerator4, Coordinate Space=::Simulation, Origin Position=$Particles.Position)
          Vector B=$Particles.SpriteFacing
         ))
        Input Min=0.2
        Input Max=0.9
        Output Min=&MergeHelper_FloatSwitch(0=0, 1=0.4, 2=0.26, 3=?, Index=&SpawnGroup(), Num=::NewEnumerator1)
        Output Max=1
       )
      )
      Y=1
     )
    )
    Scale Sprite Size Mode=::NewEnumerator2
   )
   &SolveForcesAndVelocity(Clamp Velocity=false, Limit Acceleration=false)
   &ScaleColor(
    Scale Alpha=&MergeHelper_FloatSwitch(
     0=&FloatFromCurve(
      FloatCurve=DataInterface() {propertyValues {Curve {
         defaultValue=1
         keys [
          {interpMode=RCIM_Cubic, tangentMode=RCTM_User, time=0, value=0}
          {interpMode=RCIM_Cubic, tangentMode=RCTM_User, time=0.5, value=1}
          {tangentMode=RCTM_User, time=1, value=0}
         ]
        }}}
      Scale Curve=&FNTimeOfDayScalar(Night Scale=0.7, Morning Scale=0.8, Evening Scale=0.8)
     )
     1=&FNTimeOfDayScalar(Night Scale=0.65, Morning Scale=0.7, Evening Scale=0.7)
     2=&FNTimeOfDayScalar(Night Scale=0.65, Morning Scale=0.7, Evening Scale=0.7)
     3=?
     Index=&SpawnGroup()
     Num=::NewEnumerator1
    )
    ScaleA=true
    ScaleRGB=true
   )
  ]
  rendererClasses [&NiagaraMeshRendererProperties, &NiagaraSpriteRendererProperties]
  renderers [
   {rendererClass=&NiagaraMeshRendererProperties, rendererIndex=0}
   {rendererClass=&NiagaraSpriteRendererProperties, rendererIndex=1}
   {rendererClass=&NiagaraSpriteRendererProperties, rendererIndex=2}
  ]
 }
 Shape_Light {
  emitterUpdateScript [
   &EmitterState(Life Cycle Mode=::NewEnumerator1, Loop Duration=1.8, UseLoopDelay=false)
   &SpawnBurst_Instantaneous(Use Spawn Probability=false, UseLoopCountLimit=false)
  ]
  particleSpawnScript [
   /SetVariables(Particles.Custom Random:NiagaraFloat=&RandomRangeFloat(Minimum=0.2, Maximum=0.8)) {isSetParameters=true}
   &InitializeParticle(Color Mode=::NewEnumerator0, Lifetime=2.1, Position Offset=<0 -60 75>, UsePositionOffset=true, Write Material Random=true)
   &ApplyInitialForces()
  ]
  particleUpdateScript [
   &ParticleState()
   /SetVariables(Particles.Flicker:NiagaraFloat=&Waveform(
     "Waveform [1]"=::NewEnumerator0
     "[1] Frequency"=&Lerp_Float(A=0.5, B=2.5, Alpha=$Particles.Custom Random)
     "[1] Amplitude Min/Max"=<0 1>
    )) {isSetParameters=true}
   &FN_Light_Attributes(
    Color=&MakeLinearColorFromVectorAndFloat(
     "Vector (RGB)"=&MakeVectorFromLinearColorRGB(Color=&LerpLinearColors(StartColor=<rgba 0 0.8308 1 1>, EndColor=<rgba 1 0.3231 0 1>, LerpFactor=$Particles.Flicker))
     "Float (Alpha)"=&FloatFromCurve(FloatCurve=DataInterface() {propertyValues {Curve {
         defaultValue=0
         keys [
          {arriveTangent=4.9613, interpMode=RCIM_Cubic, leaveTangent=4.9613, tangentMode=RCTM_User, time=0, value=0}
          {
           arriveTangent=-0.0086
           arriveTangentWeight=0.0333
           interpMode=RCIM_Cubic
           leaveTangent=-0.0086
           leaveTangentWeight=0.2667
           tangentMode=RCTM_User
           time=0.1647
           value=1
          }
          {arriveTangent=-0.3273, interpMode=RCIM_Cubic, leaveTangent=-0.3273, tangentMode=RCTM_User, time=0.7853, value=1}
          {arriveTangent=-0.6349, tangentMode=RCTM_User, time=0.9924, value=0}
         ]
        }}})
    )
    Create Color Attribute=true
    Create Exponent Attribute=true
    Create Radius Attribute=true
    Enable Particle Light Enable=true
    Enable Visibility Tag=true
    Exponent:NiagaraFloat=&FNTimeOfDayScalar(Night Scale=1.15, Morning Scale=1, Day Scale=0.75, Evening Scale=1)
    Light Enabled=true
    Radius=100
    Scale Color=3
   )
  ]
  rendererClasses [&NiagaraLightRendererProperties]
  renderers [{rendererClass=&NiagaraLightRendererProperties, rendererIndex=0}]
 }
 Shapes {
  emitterUpdateScript [
   &EmitterState(
    Enable Distance Culling=true
    Min Distance Response=::SleepAndClearParticles
    MinDistance=225
    Module.bUseMaxDistance=false
    Module.bUseMinDistance=true
    Scalability Mode=::NewEnumerator1
   )
   &SpawnBurst_Instantaneous(Spawn Time=0.055, Use Spawn Probability=false, UseLoopCountLimit=false)
   &SpawnBurst_Instantaneous(Spawn Group=2, Spawn Time=0.05, Use Spawn Probability=false, UseLoopCountLimit=false) {module=SpawnBurst_Instantaneous002}
   &SpawnRate(SpawnRate=1.8, Use Spawn Probability=false)
   &SpawnRate(SpawnGroup=1, SpawnRate=2.3, Use Spawn Probability=false) {module=SpawnRate001}
   &SpawnRate(SpawnGroup=2, SpawnRate=3, Use Spawn Probability=false) {module=SpawnRate002}
  ]
  particleSpawnScript [
   /SetVariables(Particles.VisibilityTag=&SpawnGroup()) {isSetParameters=true}
   &PartitionParticles(Number Of Partitions=3)
   /SetVariables(Particles.PartitionParticles.Partition:NiagaraInt32=&SpawnGroup()) {isSetParameters=true}
   &InitializeParticle(
    Color=&SelectLinearColorFromArray(
     Array Sampling Mode=::NewEnumerator0
     Direct Array Index=?
     Color Selection Array=DataInterface() {propertyValues {ColorData [<rgba 1 0.402 0 1>, <rgba 0 0.7835 1 1>, <rgba 1 0.7011 0 1>, <rgba 0 0.4287 1 1>]}}
     Random Seed=0
    )
    Lifetime=&MergeHelper_FloatSwitch(
     0=&RandomRangeFloat(Minimum=0.5, Maximum=0.85)
     1=&RandomRangeFloat(Minimum=0.95, Maximum=1.1)
     2=&RandomRangeFloat(Minimum=1.5, Maximum=1.6)
     3=?
     Index=$Particles.PartitionParticles.Partition
     Num=::NewEnumerator1
    )
    Position Offset=<0 -77 20>
    Sprite Rotation Mode=::NewEnumerator2
    Sprite Size=&MergeHelper_Vector2Switch(
     0=&RandomRangeVector2D(Minimum=<120 120>, Maximum=<160 160>)
     1=&RandomRangeVector2D(Minimum=<75 75>, Maximum=<120 120>)
     2=&RandomRangeVector2D(Minimum=<40 40>, Maximum=<90 90>)
     3=?
     4=?
     Index=$Particles.PartitionParticles.Partition
     Num=::NewEnumerator1
    )
    Sprite Size Mode=::NewEnumerator4
    UsePositionOffset=true
    Write Material Random=true
   )
   &ShapeLocation(
    Box Size=&MergeHelper_VectorSwitch(0=<100 30 0>, 1=<165 50 1>, 2=<185 65 1>, 3=?, Index=$Particles.PartitionParticles.Partition, Num=::NewEnumerator1)
    MaskSpawnGroup=false
    Offset=&MergeHelper_VectorSwitch(0=<0 8 0>, 1=<0 -3 0>, 2=<0 -8 0>, 3=?, Index=$Particles.PartitionParticles.Partition, Num=::NewEnumerator1)
    Offset Mode=::NewEnumerator0
    Override Seed=false
    Shape Primitive=::NewEnumerator2
    Surface Only=false
    UseOwnerScale=true
   )
   &AddVelocity(Velocity=&MergeHelper_VectorSwitch(
     0=&RandomRangeVector(Minimum=<0 0 150>, Maximum=<0 0 180>)
     1=&RandomRangeVector(Minimum=<0 0 80>, Maximum=<0 0 150>)
     2=&RandomRangeVector(Minimum=<0 0 70>, Maximum=<0 0 110>)
     3=?
     Index=$Particles.PartitionParticles.Partition
     Num=::NewEnumerator1
    ))
   &ApplyInitialForces()
   &CameraOffset(Camera Offset Amount=&MergeHelper_FloatSwitch(0=13, 1=-1, 2=-7, 3=?, Index=&SpawnGroup(), Num=::NewEnumerator1))
  ]
  particleUpdateScript [
   &ParticleState()
   &ScaleSpriteSize(
    Non-Uniform Curve Sprite Scale=DataInterface() {propertyValues {
      XCurve {keys [
        <at 0 0 RCIM_Cubic>
        {arriveTangentWeight=0.0527, interpMode=RCIM_Cubic, leaveTangentWeight=0.3932, tangentMode=RCTM_User, time=0.1049, value=1}
        {arriveTangentWeight=0.3916, interpMode=RCIM_Cubic, leaveTangentWeight=0.0558, tangentMode=RCTM_User, time=0.8884, value=1}
        <at 1 0 RCIM_Cubic>
       ]}
      YCurve {keys [
        <at 0 0 RCIM_Cubic>
        {arriveTangentWeight=0.0201, interpMode=RCIM_Cubic, leaveTangentWeight=0.4437, tangentMode=RCTM_User, time=0.049, value=1}
        {arriveTangentWeight=0.446, interpMode=RCIM_Cubic, leaveTangentWeight=0.0156, tangentMode=RCTM_User, time=0.9554, value=1}
        <at 1 0 RCIM_Cubic>
       ]}
     }}
    Scale Sprite Size Mode=::NewEnumerator3
   )
   &CurlNoiseForce(
    Bias Noise=false
    Noise Frequency=&RandomRangeFloat(Minimum=25, Maximum=50)
    Noise Strength=&RandomRangeFloat(Minimum=20, Maximum=50)
    Pan Noise=false
   )
   &SolveForcesAndVelocity(Clamp Velocity=false, Limit Acceleration=false)
   &ScaleColor(
    Scale Alpha=&Multiply_Float(
     A=&UniformAOrBFloat(
      A=&FloatFromCurve(FloatCurve=DataInterface() {propertyValues {Curve {keys [
           <at 0.005 0 RCIM_Cubic>
           {arriveTangent=1.1256, interpMode=RCIM_Cubic, leaveTangent=1.1256, time=0.1918, value=1}
           {arriveTangent=-0.4067, interpMode=RCIM_Cubic, leaveTangent=-0.4067, time=0.4576, value=0.5095}
           {arriveTangent=-0.9393, interpMode=RCIM_Cubic, leaveTangent=-0.9393, time=0.6474, value=0.8147}
           <at 1 0 RCIM_Cubic>
          ]}}})
      B=&FloatFromCurve(FloatCurve=DataInterface() {propertyValues {Curve {keys [
           <at 0.005 0 RCIM_Cubic>
           {arriveTangent=2.0693, interpMode=RCIM_Cubic, leaveTangent=2.0693, time=0.1233, value=0.608}
           {arriveTangent=1.1449, interpMode=RCIM_Cubic, leaveTangent=1.1449, time=0.2684, value=0.545}
           {arriveTangent=-0.0642, interpMode=RCIM_Cubic, leaveTangent=-0.0642, time=0.4451, value=0.9765}
           {arriveTangent=-0.994, interpMode=RCIM_Cubic, leaveTangent=-0.994, time=0.618, value=0.5226}
           {arriveTangent=-1.3679, interpMode=RCIM_Cubic, leaveTangent=-1.3679, time=0.7607, value=0.6627}
           <at 1 0 RCIM_Cubic>
          ]}}})
      Evaluation Type=::NewEnumerator0
      Recalculate Random Each Loop=true
      Override Seed=false
     )
     B=&FloatFromCurve(FloatCurve=DataInterface() {propertyValues {Curve {keys [
          {arriveTangent=1.2569, interpMode=RCIM_Cubic, leaveTangent=1.2569, tangentMode=RCTM_User, time=0, value=0}
          {arriveTangentWeight=0.1982, interpMode=RCIM_Cubic, leaveTangentWeight=0.2015, tangentMode=RCTM_User, time=0.328, value=1}
          {arriveTangentWeight=0.1993, interpMode=RCIM_Cubic, leaveTangentWeight=0.2023, tangentMode=RCTM_User, time=0.6795, value=1}
          {arriveTangent=-1.8646, interpMode=RCIM_Cubic, leaveTangent=-1.8646, tangentMode=RCTM_User, time=1, value=0}
         ]}}})
    )
    ScaleA=true
    ScaleRGB=true
   )
   &SubUVAnimation(
    Evaluation Type=::NewEnumerator0
    "Sprite Renderer "=DataInterface() {propertyValues {SpriteRenderer=&self:Shapes_0.Renderer}}
    SubUV Animation Mode:ENiagara_SubUVLookupModeV2=::NewEnumerator2
    UseEndFrame=false
    UseStartFrame=false
   )
   /SetVariables(
    Particles.Glow Color=&MakeLinearColorFromVectorAndFloat(
     "Vector (RGB)"=&MakeVectorFromLinearColorRGB(Color=$Particles.Color)
     "Float (Alpha)"=&Multiply_Float(A=0.25, B=&FNTimeOfDayScalar(Night Scale=0.65, Morning Scale=0.7, Evening Scale=0.7))
    )
    Particles.Glow Size=&Multiply_Vector2D_ByFloat(Float=1.5, Vector2D=$Particles.SpriteSize)
   ) {isSetParameters=true}
  ]
  rendererClasses [&NiagaraSpriteRendererProperties]
  renderers [
   {rendererClass=&NiagaraSpriteRendererProperties}
   {rendererClass=&NiagaraSpriteRendererProperties, rendererIndex=1}
   {rendererClass=&NiagaraSpriteRendererProperties, rendererIndex=2}
   {rendererClass=&NiagaraSpriteRendererProperties, rendererIndex=3}
  ]
 }
 Static_Sparkles {
  emitterUpdateScript [
   &EmitterState()
   &SpawnRate(SpawnRate=5, Use Spawn Probability=false)
   &SpawnRate(SpawnGroup=1, SpawnRate=5, Use Spawn Probability=false) {module=SpawnRate001}
   &SpawnRate(SpawnGroup=2, SpawnRate=5, Use Spawn Probability=false) {module=SpawnRate002}
   &SpawnRate(SpawnGroup=3, SpawnRate=2, Use Spawn Probability=false) {module=SpawnRate003}
   &SpawnRate(SpawnGroup=4, SpawnRate=1.75, Use Spawn Probability=false) {module=SpawnRate004}
  ]
  particleSpawnScript [
   /SetVariables(Particles.VisibilityTag=&SpawnGroup()) {isSetParameters=true}
   &InitializeParticle(
    Color=&MergeHelper_ColorSwitch(
     0=&RandomRangeLinearColor(Minimum=<rgba 0.4942 0.5473 1.25 1>, Maximum=<rgba 0.5227 0.9547 1.25 1>)
     1=&RandomRangeLinearColor(Minimum=<rgba 0.3796 0.453 1.25 1>, Maximum=<rgba 0.2369 1.0942 1.25 1>)
     2=&RandomRangeLinearColor(Minimum=<rgba 0.3796 0.453 1.25 1>, Maximum=<rgba 0.2369 1.0942 1.25 1>)
     3=&RandomRangeLinearColor(Minimum=<rgba 0.3796 0.453 1.25 1>, Maximum=<rgba 0.2369 1.0942 1.25 1>)
     4=&RandomRangeLinearColor(Minimum=<rgba 0.3796 0.453 1.25 1>, Maximum=<rgba 0.2369 1.0942 1.25 1>)
     5=&RandomRangeLinearColor(Minimum=<rgba 1 1 1 1>, Maximum=<rgba 0.2917 0.7498 1 1>)
     6=&RandomRangeLinearColor(Minimum=<rgba 1 1 1 1>, Maximum=<rgba 0.2917 0.7498 1 1>)
     7=&RandomRangeLinearColor(Minimum=<rgba 1 1 1 1>, Maximum=<rgba 0.2917 0.7498 1 1>)
     Index=&SpawnGroup()
     Num=::NewEnumerator3
    )
    Lifetime=&MergeHelper_FloatSwitch(
     0=&RandomRangeFloat(Minimum=0.3, Maximum=0.6)
     1=&RandomRangeFloat(Minimum=0.3, Maximum=0.6)
     2=&RandomRangeFloat(Minimum=0.3, Maximum=0.6)
     3=&RandomRangeFloat(Minimum=0.3, Maximum=0.6)
     4=&RandomRangeFloat(Minimum=0.3, Maximum=0.6)
     5=&RandomRangeFloat(Minimum=2.75, Maximum=3.15)
     6=&RandomRangeFloat(Minimum=2.75, Maximum=3.15)
     7=&RandomRangeFloat(Minimum=2.75, Maximum=3.15)
     Index=&SpawnGroup()
     Num=::NewEnumerator3
    )
    Sprite Size Mode=::NewEnumerator3
    Uniform Sprite Size=&MergeHelper_FloatSwitch(
     0=&RandomRangeFloat(Minimum=8, Maximum=28)
     1=&RandomRangeFloat(Minimum=8, Maximum=28)
     2=&RandomRangeFloat(Minimum=17, Maximum=38)
     3=&RandomRangeFloat(Minimum=70, Maximum=90)
     4=&RandomRangeFloat(Minimum=50, Maximum=80)
     5=&RandomRangeFloat(Minimum=8, Maximum=17)
     6=&RandomRangeFloat(Minimum=10, Maximum=28)
     7=&RandomRangeFloat(Minimum=100, Maximum=25)
     Index=&SpawnGroup()
     Num=::NewEnumerator3
    )
    UsePositionOffset=false
    Write Material Random=true
   )
   &ShapeLocation(
    Handle Radius:NiagaraFloat=&MergeHelper_FloatSwitch(0=30, 1=30, 2=0, 3=0, 4=0, Index=&SpawnGroup(), Num=::NewEnumerator3)
    Large Radius=&MergeHelper_FloatSwitch(0=50, 1=50, 2=0, 3=0, 4=0, Index=&SpawnGroup(), Num=::NewEnumerator3)
    MaskSpawnGroup=false
    Non Uniform Scale=<1 1 2.5>
    Offset=<0 0 16>
    Offset Mode=::NewEnumerator0
    Override Seed=false
    Rotation Mode=::NewEnumerator2
    Shape Primitive=::NewEnumerator3
    UseOwnerScale=true
    Yaw / Pitch / Roll=<1.6 0 0>
   )
   &ShapeLocation(
    Cylinder Height=&MergeHelper_FloatSwitch(0=0, 1=0, 2=150, 3=150, 4=150, Index=&SpawnGroup(), Num=::NewEnumerator3)
    Cylinder Radius=&MergeHelper_FloatSwitch(0=0, 1=0, 2=20, 3=20, 4=20, Index=&SpawnGroup(), Num=::NewEnumerator3)
    MaskSpawnGroup=false
    Non Uniform Scale=<1 1.4 1>
    Offset=&SelectVectorFromArray(
     Array Sampling Mode=::NewEnumerator0
     Direct Array Index=?
     Vector Selection Array=DataInterface() {propertyValues {FloatData [<60 0 90>, <-60 0 90>]}}
     Random Seed=0
    )
    Offset Mode=::NewEnumerator0
    Override Seed=false
    Shape Primitive=::NewEnumerator1
    Surface Only=false
    UseOwnerScale=true
   ) {module=ShapeLocation001}
   &ApplyInitialForces()
  ]
  particleUpdateScript [
   &ParticleState()
   /SetVariables(
    Particles.Update Large 1 Color:LinearColor=&MultiplyLinearColor(
     Linear Color=$Particles.Color
     Scale Factor=&MakeVector4_FromLinearColor(LinearColor=<rgba 0.9441 0.8079 1 1>)
    )
    Particles.Update Large 3 Color:LinearColor=&MultiplyLinearColor(
     Linear Color=$Particles.Color
     Scale Factor=&MakeVector4_FromLinearColor(LinearColor=<rgba 0.15 0.15 0.15 1>)
    )
    Particles.Update Large Color:LinearColor=&MultiplyLinearColor(
     Linear Color=$Particles.Color
     Scale Factor=&MakeVector4_FromLinearColor(LinearColor=<rgba 8.0099 9.8007 10 1>)
    )
    Particles.Update Small Static Color:LinearColor=&MultiplyLinearColor(
     Linear Color=$Particles.Color
     Scale Factor=&MakeVector4_FromLinearColor(LinearColor=<rgba 6.5985 9.9045 10 1>)
    )
    Particles.Update Small Static Color 1:LinearColor=&MultiplyLinearColor(
     Linear Color=$Particles.Color
     Scale Factor=&MakeVector4_FromLinearColor(LinearColor=<rgba 1 0.9854 0.8058 1>)
    )
   ) {isSetParameters=true}
   &ScaleColor(
    Scale Alpha=&Multiply_Float(
     A=&FloatFromCurve(
      FloatCurve=DataInterface() {propertyValues {Curve {keys [
          {interpMode=RCIM_Cubic, tangentMode=RCTM_User, time=0, value=0}
          {arriveTangent=-2.4058, interpMode=RCIM_Cubic, leaveTangent=-2.4058, tangentMode=RCTM_User, time=0.26, value=0.8421}
          {arriveTangent=-1.1775, interpMode=RCIM_Cubic, leaveTangent=-1.1775, time=0.4613, value=0.0645}
          {arriveTangent=-0.1197, interpMode=RCIM_Cubic, leaveTangent=-0.1197, time=0.5898, value=0.4537}
          <at 1 0 RCIM_Cubic>
         ]}}}
      Scale Curve=&RandomRangeFloat(Minimum=0.1, Maximum=0.5)
     )
     B=&FloatFromCurve(
      FloatCurve=DataInterface() {propertyValues {Curve {keys [{tangentMode=RCTM_User, time=0, value=0}, {tangentMode=RCTM_User, time=1, value=1}]}}}
      CurveIndex=&Multiply_Float(
       A=&MakeFloatFromVector(Vector=Linked() {linkedVariable {name=Particles.Position, type=NiagaraPosition}}, Channel=::NewEnumerator2)
       B=$Engine.System.Age
      )
     )
    )
    Scale RGB=&MakeVectorFromLinearColorRGB(Color=&FNTimeOfDayColor(Night Color=<rgba 0.3 0.6267 1 10>, Morning Color=<rgba 1 1 1 1>, Day Color=<rgba 1 1 1 1>, Evening Color=<rgba 1 1 1 1>))
    ScaleA=true
    ScaleRGB=true
   )
  ]
  rendererClasses [&NiagaraSpriteRendererProperties]
  renderers [
   {rendererClass=&NiagaraSpriteRendererProperties}
   {rendererClass=&NiagaraSpriteRendererProperties, rendererIndex=1}
   {rendererClass=&NiagaraSpriteRendererProperties, rendererIndex=2}
   {rendererClass=&NiagaraSpriteRendererProperties, rendererIndex=3}
   {rendererClass=&NiagaraSpriteRendererProperties, rendererIndex=4}
  ]
 }
 Top_Left {rendererClasses [], renderers []}
 Top_Right {rendererClasses [], renderers []}
}
summary {
 systemName=NS_Emotes_StoneFly_Shapes
 emitters [
  {emitterName=Shape_Light, rendererClasses [&NiagaraLightRendererProperties]}
  {emitterName=Fake_Shape_Lights, rendererClasses []}
  {emitterName=Glows_Air, rendererClasses []}
  {emitterName=Glows_Decal, rendererClasses []}
  {emitterName=Rising_Sparkles, rendererClasses [&NiagaraSpriteRendererProperties]}
  {emitterName=Shapes, rendererClasses [&NiagaraSpriteRendererProperties]}
  {emitterName=Static_Sparkles, rendererClasses [&NiagaraSpriteRendererProperties]}
  {emitterName=Top_Left, rendererClasses []}
  {emitterName=Top_Right, rendererClasses []}
  {emitterName=Screen, rendererClasses [&NiagaraMeshRendererProperties, &NiagaraSpriteRendererProperties]}
  {emitterName=Legs_Light, rendererClasses []}
 ]
 userVariables [
  User.bIsFrontEnd {type=NiagaraBool}
  User.bIsFrontEndPreview {type=NiagaraBool}
  User.Color_Index {type=NiagaraInt32}
  User.OwnerLocalVelocity {type=Vector3f, description="", defaultValue {struct=Vector3f, value=<0 0 0>}}
  User.WipeDuration {type=NiagaraFloat, defaultValue=0.8, description=""}
 ]
}
